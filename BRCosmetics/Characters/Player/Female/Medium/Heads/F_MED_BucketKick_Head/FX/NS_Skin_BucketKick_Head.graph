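# Niagara cosmetic effect for the BucketKick female character head, dressing hair, eyes and neck in pink-purple animated meshes, with flame-bit sprites and a pulsing glow light.
# Burst-spawned meshes follow skeletal mesh points using per-index arrays for placement and shape, most with a second, larger render layer.
# A time-smoothed position is packed with intensity into a dynamic parameter, apparently for motion lag.
# BigHeadMode rescales most parts, and bisHeadHidden drives visibility and light.
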NiagaraSystem /BRCosmetics/Characters/Player/Female/Medium/Heads/F_MED_BucketKick_Head/FX/NS_Skin_BucketKick_Head
dependencies {
 usedDataInterfaces [
  &NiagaraDataInterfaceArrayFloat
  &NiagaraDataInterfaceArrayFloat3
  &NiagaraDataInterfaceCurve
  &NiagaraDataInterfaceColorCurve
  &NiagaraDataInterfaceSkeletalMesh
  &NiagaraDataInterfaceSocketReader
 ]
 usedDynamicInputs [
  &MakeCustomFloatFromBool
  &SetBoolByFloatComparison
  &VectorLength
  &ClampFloat
 ]
 usedModules [
  &self:SystemSpawnScript.SystemScriptSource.SystemScriptGraph.NiagaraNodeAssignment_3.SetVariables_#1
  &self:SystemSpawnScript.SystemScriptSource.SystemScriptGraph.NiagaraNodeAssignment_1.SetVariables_#2
  &self:SystemSpawnScript.SystemScriptSource.SystemScriptGraph.NiagaraNodeAssignment_2.SetVariables_#3
  &SystemState
  &self:SystemSpawnScript.SystemScriptSource.SystemScriptGraph.NiagaraNodeAssignment_0.SetVariables_#4
 ]
}
emitters {
 Eye {
  emitterUpdateScript [
   &EmitterState(Life Cycle Mode=::NewEnumerator1, Loop Behavior=::NewEnumerator1, Loop Duration Mode=::NewEnumerator1, UseLoopDelay=false)
   &SpawnBurst_Instantaneous(Spawn Count=2, Use Spawn Probability=false, UseLoopCountLimit=false)
  ]
  particleSpawnScript [
   &InitializeParticle(
    Color=<rgba 1 0.2795 0.653 1>
    Mesh Scale Mode=::NewEnumerator3
    Mesh Uniform Scale=&Multiply_Float(
     A=1
     B=&SelectFloatFromArray(
      Direct Array Index=&ReturnExecIndex()
      Float Selection Array=DataInterface() {propertyValues {FloatData [1.35,0.8,0.8,0.6,0.6]}}
     )
    )
    UsePositionOffset=false
    Write Material Random=true
   )
   &InitialMeshOrientation(
    Mesh Axis To Orient=<0 0 1>
    Mesh Orientation Mode=::NewEnumerator2
    Orientation Vector=&SelectVectorFromArray(
     Direct Array Index=&ReturnExecIndex()
     Vector Selection Array=DataInterface() {propertyValues {FloatData [<1 0 -2>, <1 0 -2>]}}
    )
    Use Rotation Vector=false
   )
   /SetVariables(
    Particles.R2_Color:LinearColor=<rgba 1 0 0.4853 1>
    Particles.R2_Scale:Vector3f=&Multiply_Vector(A=$Particles.Scale, B=<1.25 1.25 1.25>)
   ) {isSetParameters=true, module=SetVariables_#5}
   /SetVariables(Particles.RandU:NiagaraFloat=&RandomRangeFloat(Minimum=0, Maximum=1)) {isSetParameters=true, module=SetVariables_#6}
  ]
  particleUpdateScript [
   &ParticleState(Kill Particles When Lifetime Has Elapsed=false, Let Infinitely Lived Particles Die When Emitter Deactivates=true)
   &ScaleMeshSize(Scale Factor=&VectorFromFloat(Value=&Lerp_Float(A=1, B=2, Alpha=$NPC.FortniteNPC.BigHeadMode)))
   &SkeletalMeshLocation(
    MaskSpawnGroup=false
    Mesh Sampling Type=::NewEnumerator5
    Sampled Position Offset=&Multiply_VectorByFloat(
     Vector=&SelectVectorFromArray(
      Direct Array Index=&ReturnExecIndex()
      Vector Selection Array=DataInterface() {propertyValues {FloatData [<0 -0.5 0>, <-1 1.5 2>]}}
     )
     Float=&Lerp_Float(A=1, B=2, Alpha=$NPC.FortniteNPC.BigHeadMode)
    )
    Skeletal Mesh=$System.Skeletal Mesh
    Socket Sampling Mode=::NewEnumerator5
   )
   &InterpolateOverTime(
    Rate Of Change:NiagaraFloat=$System.Interpolate Rate
    Target Vector Value:NiagaraPosition=$Particles.Position
    Variable Type=::NewEnumerator2
   )
   &DynamicMaterialParameters(
    Index 0 Param 1=&Multiply_Float(
     A=&SelectFloatFromArray(
      Direct Array Index=&ReturnExecIndex()
      Float Selection Array=DataInterface() {propertyValues {FloatData [1.7,1.2]}}
     )
     B=&Lerp_Float(A=1, B=1.75, Alpha=$NPC.FortniteNPC.BigHeadMode)
    )
    Index 0 Param 2=&Multiply_Float(A=$Engine.Time, B=1)
    Index 0 Param 3=&SelectFloatFromArray(
     Direct Array Index=&ReturnExecIndex()
     Float Selection Array=DataInterface() {propertyValues {FloatData [0.4,0.3]}}
    )
    Index 0 Param 4=$Particles.RandU
    Index 2 Param 1=1
    Module.Index0Param0Write=true
    Module.Index0Param1Write=true
    Module.Index0Param2Write=true
    Module.Index0Param3Write=true
    Module.Index2Param0Write=true
    Module.Index2Param1Write=true
    Module.Index2Param2Write=true
    Module.Index2Param3Write=true
    Write Parameter Index 2=true
   )
   /SetVariables(Particles.Interpolate DY:Vector4f=&MakeFloat4FromVectorAndFloat("Vector (RGB)"=$Particles.InterpolateOverTime.InterpolateOverTime_V3_.Moving Average, "Float (Alpha)"=$System.Interpolate Internsity)) {isSetParameters=true, module=SetVariables_#7}
   /SetVariables(Particles.R2_DynMatParm:Vector4f=&MakeVector4(
     X=&Multiply_Float(
      A=1.25
      B=&SelectFloatFromArray(
       Direct Array Index=&ReturnExecIndex()
       Float Selection Array=DataInterface() {propertyValues {FloatData [1.7,1.2]}}
      )
     )
     Y=&Multiply_Float(A=$Engine.Time, B=1)
     Z=&SelectFloatFromArray(
      Direct Array Index=&ReturnExecIndex()
      Float Selection Array=DataInterface() {propertyValues {FloatData [0.4,0.3]}}
     )
     W=$Particles.RandU
    )) {isSetParameters=true, module=SetVariables_#8}
   /SetVariables() {isSetParameters=true, module=SetVariables_#9}
   /SetVariables(Particles.VisibilityTag=&MakeCustomIntFromBool(Bool=$User.bisHeadHidden, True INT=1, False INT=0)) {isSetParameters=true, module=SetVariables_#10}
  ]
  rendererClasses [&NiagaraMeshRendererProperties]
  renderers [
   {rendererClass=&NiagaraMeshRendererProperties, rendererIndex=0}
   {rendererClass=&NiagaraMeshRendererProperties, rendererIndex=1}
  ]
 }
 Flame_bits {
  emitterUpdateScript [
   &EmitterState()
   &SpawnRate(SpawnRate=7, Use Spawn Probability=false)
  ]
  particleSpawnScript [
   &InitializeParticle(
    Lifetime Max=1.3
    Lifetime Min=0.5
    Lifetime Mode=::NewEnumerator1
    Sprite Size Max=<0.7 5>
    Sprite Size Min=<0.5 3>
    Sprite Size Mode=::NewEnumerator2
    Sprite UV Mode=::NewEnumerator3
    UsePositionOffset=false
    Write Material Random=true
   )
   &SocketLocation(
    SamplePosOffset=true
    Sampled Position Offset=<10 -12 0>
    Socket Reader=$System.Socket Reader Interface
    Socket Sampling Mode=::NewEnumerator3
    VelocitySpeedLimit=false
   )
   &AddVelocity(Cone Axis=<0.5 -2 1>, Module.UseConeFalloff=false, Module.UseVelDistribution=false, Override Seed=false, Velocity Mode=::NewEnumerator2)
   &ShapeLocation(
    MaskSpawnGroup=false
    Override Seed=false
    Rotation Angle=-0.1753
    Rotation Axis=<0 1 0>
    Rotation Mode=::NewEnumerator1
    Sphere Radius=10
    UseOwnerScale=true
   )
   /SetVariables(Particles.RandU:NiagaraFloat=&RandomRangeFloat(Minimum=0, Maximum=1)) {isSetParameters=true}
  ]
  particleUpdateScript [
   &ParticleState()
   &ScaleSpriteSize(Uniform Curve Sprite Scale=DataInterface() {propertyValues {Curve {keys [
        {interpMode=RCIM_Cubic, tangentMode=RCTM_User, time=0, value=1}
        {arriveTangent=-1, interpMode=RCIM_Cubic, leaveTangent=-1, time=0.5, value=1}
        {arriveTangent=-2.4284, interpMode=RCIM_Cubic, leaveTangent=-2.4284, tangentMode=RCTM_User, time=1, value=0}
       ]}}})
   /Niagara/Modules/Update/Forces/CurlNoiseForce(Noise Frequency=5.0345, Noise Strength=100, Pan Noise=true, Pan Noise Field=<1 0 0>)
   &SolveForcesAndVelocity(Clamp Velocity=false, Limit Acceleration=false)
   &SpriteRotationRate(Rotation Rate=&RandomRangeFloat(Minimum=-400, Maximum=400)) {enabled=false, module=SpriteRotationRate001}
   &ScaleColor(
    Scale RGB=&MakeVectorFromLinearColorRGB(Color=&ColorFromCurve(ColorCurve=DataInterface() {propertyValues {
        AlphaCurve {keys []}
        BlueCurve {keys [<at 0 2 RCIM_Cubic>,<at 0.47 1>]}
        GreenCurve {keys [<at 0 0 RCIM_Cubic>,<at 0.47 0>]}
        RedCurve {keys [<at 0 1.8276 RCIM_Cubic>,<at 0.47 0.1853>]}
       }}))
    ScaleA=true
    ScaleRGB=true
   )
   &DynamicMaterialParameters(
    Index 0 Param 1=$Particles.RandU
    Module.Index0Param0Write=true
    Module.Index0Param1Write=false
    Module.Index0Param2Write=false
    Module.Index0Param3Write=false
   )
   /SetVariables(Particles.VisibilityTag=&MakeCustomIntFromBool(Bool=$User.bisHeadHidden, True INT=1, False INT=0)) {isSetParameters=true, module=SetVariables_#10}
  ]
  rendererClasses [&NiagaraSpriteRendererProperties]
  renderers [{rendererClass=&NiagaraSpriteRendererProperties}]
 }
 Glow {
  emitterUpdateScript [
   &EmitterState(
    Inactive Response=::NewEnumerator1
    Life Cycle Mode=::NewEnumerator1
    Loop Behavior=::NewEnumerator1
    Loop Duration Mode=::NewEnumerator1
    UseLoopDelay=false
   )
   &SpawnBurst_Instantaneous(Use Spawn Probability=false, UseLoopCountLimit=false)
  ]
  particleSpawnScript [&InitializeParticle(
    Color=<rgba 0.5561 0 1 0.5>
    Lifetime=5
    Sprite Size Mode=::NewEnumerator3
    Uniform Sprite Size=47
    UsePositionOffset=false
    Write Material Random=true
   )]
  particleUpdateScript [
   &ParticleState(Kill Particles When Lifetime Has Elapsed=false, Let Infinitely Lived Particles Die When Emitter Deactivates=true)
   &SocketLocation(
    SamplePosOffset=true
    Sampled Position Offset=&Lerp_Vector(A=<10 -12 0>, B=<20 -24 0>, Alpha=$NPC.FortniteNPC.BigHeadMode)
    Socket Reader=$System.Socket Reader Interface
    Socket Sampling Mode=::NewEnumerator3
    VelocitySpeedLimit=false
   )
   &ScaleColor(
    Scale RGB=&VectorFromFloat(Value=&Waveform(Global Amplitude Scale=0.5, "Waveform [1]"=::NewEnumerator12, "[1] Frequency"=10, "[1] Amplitude Min/Max"=<0 1>, "[1] Bias"=1))
    ScaleA=true
    ScaleRGB=true
   )
   &FN_Light_Attributes(
    Color=$Particles.Color
    Create Color Attribute=true
    Create Exponent Attribute=true
    Create Position Attribute=true
    Create Radius Attribute=true
    Enable Particle Light Enable=true
    Exponent=5
    Light Enabled:NiagaraBool=&Boolean_NotOperation(Bool=$User.bisHeadHidden)
    Position=&AddVectorToPosition(Position=$Particles.Position, Vector=<0 0 0>)
    Radius=&Lerp_Float(A=40, B=80, Alpha=$NPC.FortniteNPC.BigHeadMode)
    Scale Color=30
   )
   /SetVariables(Particles.VisibilityTag=&MakeCustomIntFromBool(Bool=$User.bisHeadHidden, True INT=1, False INT=0)) {isSetParameters=true, module=SetVariables_#10}
  ]
  rendererClasses [&NiagaraSpriteRendererProperties, &NiagaraLightRendererProperties]
  renderers [
   {rendererClass=&NiagaraSpriteRendererProperties}
   {rendererClass=&NiagaraLightRendererProperties, rendererIndex=1}
  ]
 }
 Hair {
  emitterUpdateScript [
   &EmitterState(Life Cycle Mode=::NewEnumerator1, Loop Behavior=::NewEnumerator1, Loop Duration Mode=::NewEnumerator1, UseLoopDelay=false)
   &SpawnBurst_Instantaneous(Spawn Count=11, Use Spawn Probability=false, UseLoopCountLimit=false)
  ]
  particleSpawnScript [
   &InitializeParticle(
    Color=<rgba 1 0.2795 0.653 1>
    Mesh Scale Mode=::NewEnumerator3
    Mesh Uniform Scale=&Multiply_Float(
     A=3
     B=&SelectFloatFromArray(
      Direct Array Index=&ReturnExecIndex()
      Float Selection Array=DataInterface() {propertyValues {FloatData [1,0.8,0.8,0.75,0.6,0.75,0.45,0.3,0.2,0.35,0.5]}}
     )
    )
    UsePositionOffset=false
    Write Material Random=true
   )
   &SocketLocation(
    SamplePosOffset=true
    Sampled Position Offset=&SelectVectorFromArray(
     Direct Array Index=&ReturnExecIndex()
     Vector Selection Array=DataInterface() {propertyValues {FloatData [<-5 -12 -10>, <-4 -12 -7>, <-1 -13 -4>, <-1 -13 -1>, <3 -12 -1>, <-5.5 -12 -2>, <-1 -11.5 -8>]}}
    )
    Socket Reader=$System.Socket Reader Interface
    Socket Sampling Mode=::NewEnumerator3
    VelocitySpeedLimit=false
   ) {enabled=false}
   &InitialMeshOrientation(
    Mesh Axis To Orient=<0 0 1>
    Mesh Orientation Mode=::NewEnumerator2
    Orientation Vector=&SelectVectorFromArray(
     Direct Array Index=&ReturnExecIndex()
     Vector Selection Array=DataInterface() {propertyValues {FloatData [<1 -1 0>, <0.6 -1 0>, <0.6 -1 0>, <1 -0.7 0>, <1 -0.7 0>, <1 -0.4 0>, <1 -0.4 0>, <1 -1 0.5>, <0.5 -1 0.5>, <0.5 -1 -0.2>, <0.5 -1 0.1>]}}
    )
    Use Rotation Vector=false
   )
   /SetVariables(
    Particles.R2_Color:LinearColor=<rgba 1 0 0.4853 1>
    Particles.R2_Scale:Vector3f=&Multiply_Vector(A=$Particles.Scale, B=<1.25 1.25 1.25>)
   ) {isSetParameters=true, module=SetVariables_#5}
   /SetVariables(
    Particles.RandSpeed:NiagaraFloat=&RandomRangeFloat(Minimum=0.7, Maximum=1.2)
    Particles.RandU:NiagaraFloat=&RandomRangeFloat(Minimum=0, Maximum=1)
   ) {isSetParameters=true, module=SetVariables_#6}
  ]
  particleUpdateScript [
   &ParticleState(Kill Particles When Lifetime Has Elapsed=false, Let Infinitely Lived Particles Die When Emitter Deactivates=true)
   &ScaleMeshSize(Scale Factor=&VectorFromFloat(Value=&Lerp_Float(A=1, B=2, Alpha=$NPC.FortniteNPC.BigHeadMode)))
   &SkeletalMeshLocation(
    Bone Sampling Mode=::NewEnumerator4
    MaskSpawnGroup=false
    Sampled Position Offset=&Multiply_VectorByFloat(
     Vector=&SelectVectorFromArray(
      Direct Array Index=&ReturnExecIndex()
      Vector Selection Array=DataInterface() {propertyValues {FloatData [<7 -7 0>, <5 -7 -2>, <5 -7 2>, <7.5 -6 4>, <7.5 -6 -4>, <8 -4.5 -4.5>, <8 -4.5 4.5>, <2 -5 5>, <0.5 -5 4.5>, <2 -4.3 -5>, <6.5 -6 3.2>]}}
     )
     Float=&Lerp_Float(A=1, B=2, Alpha=$NPC.FortniteNPC.BigHeadMode)
    )
    Skeletal Mesh=$System.Skeletal Mesh
   )
   &InterpolateOverTime(
    Rate Of Change:NiagaraFloat=$System.Interpolate Rate
    Target Vector Value:NiagaraPosition=$Particles.Position
    Variable Type=::NewEnumerator2
   )
   &DynamicMaterialParameters(
    Index 0 Param 1=&Multiply_Float(
     A=&SelectFloatFromArray(
      Direct Array Index=&ReturnExecIndex()
      Float Selection Array=DataInterface() {propertyValues {FloatData [4.5,3.5,3.5,4,3,3.5,2.5,2,1.5,2]}}
     )
     B=&Lerp_Float(A=1, B=1.75, Alpha=$NPC.FortniteNPC.BigHeadMode)
    )
    Index 0 Param 2=&Multiply_Float(A=$Engine.Time, B=$Particles.RandSpeed)
    Index 0 Param 3=&SelectFloatFromArray(
     Direct Array Index=&ReturnExecIndex()
     Float Selection Array=DataInterface() {propertyValues {FloatData [1,0.7,0.7,0.6,0.5,0.55,0.4,0.3,0.2,0.3]}}
    )
    Index 0 Param 4=$Particles.RandU
    Index 2 Param 1=1
    Module.Index0Param0Write=true
    Module.Index0Param1Write=true
    Module.Index0Param2Write=true
    Module.Index0Param3Write=true
    Module.Index2Param0Write=true
    Module.Index2Param1Write=true
    Module.Index2Param2Write=true
    Module.Index2Param3Write=true
    Write Parameter Index 2=true
   )
   /SetVariables(Particles.Interpolate DY:Vector4f=&MakeFloat4FromVectorAndFloat("Vector (RGB)"=$Particles.InterpolateOverTime.InterpolateOverTime_V3_.Moving Average, "Float (Alpha)"=$System.Interpolate Internsity)) {isSetParameters=true, module=SetVariables_#7}
   /SetVariables(Particles.R2_DynMatParm:Vector4f=&MakeVector4(
     X=&Multiply_Float(
      A=1.2
      B=&SelectFloatFromArray(
       Direct Array Index=&ReturnExecIndex()
       Float Selection Array=DataInterface() {propertyValues {FloatData [4.5,3.5,3.5,4,3,3.5,2.5,2,1.5,2]}}
      )
     )
     Y=&Multiply_Float(A=$Engine.Time, B=$Particles.RandSpeed)
     Z=&SelectFloatFromArray(
      Direct Array Index=&ReturnExecIndex()
      Float Selection Array=DataInterface() {propertyValues {FloatData [1,0.7,0.7,0.6,0.5,0.55,0.4,0.3,0.2,0.3]}}
     )
     W=$Particles.RandU
    )) {isSetParameters=true, module=SetVariables_#8}
   /SetVariables() {isSetParameters=true, module=SetVariables_#9}
   /SetVariables(Particles.VisibilityTag=&MakeCustomIntFromBool(Bool=$User.bisHeadHidden, True INT=1, False INT=0)) {isSetParameters=true, module=SetVariables_#10}
  ]
  rendererClasses [&NiagaraMeshRendererProperties]
  renderers [
   {rendererClass=&NiagaraMeshRendererProperties, rendererIndex=0}
   {rendererClass=&NiagaraMeshRendererProperties, rendererIndex=1}
  ]
 }
 HairBig {
  emitterUpdateScript [
   &EmitterState(Life Cycle Mode=::NewEnumerator1, Loop Behavior=::NewEnumerator1, Loop Duration Mode=::NewEnumerator1, UseLoopDelay=false)
   &SpawnBurst_Instantaneous(Spawn Count=2, Use Spawn Probability=false, UseLoopCountLimit=false)
  ]
  particleSpawnScript [
   &InitializeParticle(
    Color=<rgba 0.318 0 1 0.3>
    Mesh Scale Mode=::NewEnumerator3
    Mesh Uniform Scale=&Multiply_Float(
     A=3
     B=&SelectFloatFromArray(
      Direct Array Index=&ReturnExecIndex()
      Float Selection Array=DataInterface() {propertyValues {FloatData [0.25,0.3,0.8,0.75,0.6,0.75,0.45,0.3,0.2,0.35,0.5]}}
     )
    )
    UsePositionOffset=false
    Write Material Random=true
   )
   &InitialMeshOrientation(
    Mesh Axis To Orient=<0 0 1>
    Mesh Orientation Mode=::NewEnumerator2
    Orientation Vector=&SelectVectorFromArray(
     Direct Array Index=&ReturnExecIndex()
     Vector Selection Array=DataInterface() {propertyValues {FloatData [<1 -0.17 -0.1>, <1 0 0>, <0.6 -1 0>, <1 -0.7 0>, <1 -0.7 0>, <1 -0.4 0>, <1 -0.4 0>, <1 -1 0.5>, <0.5 -1 0.5>, <0.5 -1 -0.2>, <0.5 -1 0.1>]}}
    )
    Use Rotation Vector=false
   )
   /SetVariables(
    Particles.R2_Color:LinearColor=<rgba 1 0 0.4853 1>
    Particles.R2_Scale:Vector3f=&Multiply_Vector(A=$Particles.Scale, B=<1.25 1.25 1.25>)
   ) {isSetParameters=true, module=SetVariables_#5}
   /SetVariables(
    Particles.RandSpeed:NiagaraFloat=&RandomRangeFloat(Minimum=0.7, Maximum=1.2)
    Particles.RandU:NiagaraFloat=&RandomRangeFloat(Minimum=0, Maximum=1)
   ) {isSetParameters=true, module=SetVariables_#6}
  ]
  particleUpdateScript [
   &ParticleState(Kill Particles When Lifetime Has Elapsed=false, Let Infinitely Lived Particles Die When Emitter Deactivates=true)
   &ScaleMeshSize(Scale Factor=&VectorFromFloat(Value=&Lerp_Float(A=1, B=2, Alpha=$NPC.FortniteNPC.BigHeadMode)))
   &SkeletalMeshLocation(
    Bone Sampling Mode=::NewEnumerator4
    MaskSpawnGroup=false
    Sampled Position Offset=&SelectVectorFromArray(
     Direct Array Index=&ReturnExecIndex()
     Vector Selection Array=DataInterface() {propertyValues {FloatData [<5 0 -5>, <5 0 0>, <5 -7 2>, <7.5 -6 4>, <7.5 -6 -4>, <8 -4.5 -4.5>, <8 -4.5 4.5>, <2 -5 5>, <0.5 -5 4.5>, <2 -4.3 -5>, <6.5 -6 3.2>]}}
    )
    Skeletal Mesh=$System.Skeletal Mesh
   )
   &InterpolateOverTime(
    Rate Of Change:NiagaraFloat=$System.Interpolate Rate
    Target Vector Value:NiagaraPosition=$Particles.Position
    Variable Type=::NewEnumerator2
   )
   &DynamicMaterialParameters(
    Index 0 Param 2=&Multiply_Float(A=1, B=$Engine.Time)
    Index 0 Param 3=1
    Index 0 Param 4=&RandomRangeFloat(Minimum=0, Maximum=1)
    Module.Index0Param0Write=true
    Module.Index0Param1Write=true
    Module.Index0Param2Write=true
    Module.Index0Param3Write=true
   ) {module=DynamicMaterialParameters001}
   /SetVariables(Particles.Interpolate DY:Vector4f=&MakeFloat4FromVectorAndFloat("Vector (RGB)"=$Particles.InterpolateOverTime.InterpolateOverTime_V3_.Moving Average, "Float (Alpha)"=$System.Interpolate Internsity)) {isSetParameters=true, module=SetVariables_#7}
   /SetVariables(Particles.R2_DynMatParm:Vector4f=&MakeVector4(
     X=&Multiply_Float(
      A=1.2
      B=&SelectFloatFromArray(
       Direct Array Index=&ReturnExecIndex()
       Float Selection Array=DataInterface() {propertyValues {FloatData [4.5,3.5,3.5,4,3,3.5,2.5,2,1.5,2]}}
      )
     )
     Y=&Multiply_Float(A=$Engine.Time, B=$Particles.RandSpeed)
     Z=&SelectFloatFromArray(
      Direct Array Index=&ReturnExecIndex()
      Float Selection Array=DataInterface() {propertyValues {FloatData [1,0.7,0.7,0.6,0.5,0.55,0.4,0.3,0.2,0.3]}}
     )
     W=$Particles.RandU
    )) {isSetParameters=true, module=SetVariables_#8}
   /SetVariables() {isSetParameters=true, module=SetVariables_#9}
   /SetVariables(Particles.VisibilityTag=&MakeCustomIntFromBool(Bool=$User.bisHeadHidden, True INT=1, False INT=0)) {isSetParameters=true, module=SetVariables_#10}
  ]
  rendererClasses [&NiagaraMeshRendererProperties]
  renderers [{rendererClass=&NiagaraMeshRendererProperties, rendererIndex=0}]
 }
 Neck {
  emitterUpdateScript [
   &EmitterState(Life Cycle Mode=::NewEnumerator1, Loop Behavior=::NewEnumerator1, Loop Duration Mode=::NewEnumerator1, UseLoopDelay=false)
   &SpawnBurst_Instantaneous(Spawn Count=4, Use Spawn Probability=false, UseLoopCountLimit=false)
  ]
  particleSpawnScript [
   &InitializeParticle(
    Color=<rgba 1 0.2795 0.653 1>
    Mesh Scale Mode=::NewEnumerator3
    Mesh Uniform Scale=&Multiply_Float(
     A=1
     B=&SelectFloatFromArray(
      Direct Array Index=&ReturnExecIndex()
      Float Selection Array=DataInterface() {propertyValues {FloatData [1,1.5,0.8,0.5,0.6]}}
     )
    )
    UsePositionOffset=false
    Write Material Random=true
   )
   &InitialMeshOrientation(
    Mesh Axis To Orient=<0 0 1>
    Mesh Orientation Mode=::NewEnumerator2
    Orientation Vector=&SelectVectorFromArray(
     Direct Array Index=&ReturnExecIndex()
     Vector Selection Array=DataInterface() {propertyValues {FloatData [<1 -0.5 0>, <1 -0.5 -1>, <-0.7 -0.5 -1>, <0 -1 1>]}}
    )
    Use Rotation Vector=false
   )
   /SetVariables(
    Particles.R2_Color:LinearColor=<rgba 1 0 0.4853 1>
    Particles.R2_Scale:Vector3f=&Multiply_Vector(A=$Particles.Scale, B=<1.25 1.25 1.25>)
   ) {isSetParameters=true, module=SetVariables_#5}
   /SetVariables(Particles.RandU:NiagaraFloat=&RandomRangeFloat(Minimum=0, Maximum=1)) {isSetParameters=true, module=SetVariables_#6}
  ]
  particleUpdateScript [
   &ParticleState(Kill Particles When Lifetime Has Elapsed=false, Let Infinitely Lived Particles Die When Emitter Deactivates=true)
   &ScaleMeshSize(Scale Factor=&VectorFromFloat(Value=&Lerp_Float(A=1, B=1.2, Alpha=$NPC.FortniteNPC.BigHeadMode)))
   &SkeletalMeshLocation(
    Bone / Socket Index=2
    Bone Sampling Mode=::NewEnumerator4
    MaskSpawnGroup=false
    Sampled Position Offset=&Multiply_VectorByFloat(
     Vector=&SelectVectorFromArray(
      Direct Array Index=&ReturnExecIndex()
      Vector Selection Array=DataInterface() {propertyValues {FloatData [<0 -3 0>, <0 -3 0>, <0 -3.7 0>, <0 -3.5 0>]}}
     )
     Float=&Lerp_Float(A=1, B=1.3, Alpha=$NPC.FortniteNPC.BigHeadMode)
    )
    Skeletal Mesh=$System.Skeletal Mesh
   )
   &InterpolateOverTime(
    Rate Of Change:NiagaraFloat=$System.Interpolate Rate
    Target Vector Value:NiagaraPosition=$Particles.Position
    Variable Type=::NewEnumerator2
   )
   &DynamicMaterialParameters(
    Index 0 Param 1=&SelectFloatFromArray(
     Direct Array Index=&ReturnExecIndex()
     Float Selection Array=DataInterface() {propertyValues {FloatData [1.5,2,1.3,1]}}
    )
    Index 0 Param 2=&Multiply_Float(A=$Engine.Time, B=1)
    Index 0 Param 3=&SelectFloatFromArray(
     Direct Array Index=&ReturnExecIndex()
     Float Selection Array=DataInterface() {propertyValues {FloatData [0.25,0.3,0.22,0.2]}}
    )
    Index 0 Param 4=$Particles.RandU
    Index 2 Param 1=1
    Module.Index0Param0Write=true
    Module.Index0Param1Write=true
    Module.Index0Param2Write=true
    Module.Index0Param3Write=true
    Module.Index2Param0Write=true
    Module.Index2Param1Write=true
    Module.Index2Param2Write=true
    Module.Index2Param3Write=true
    Write Parameter Index 2=true
   )
   /SetVariables(Particles.Interpolate DY:Vector4f=&MakeFloat4FromVectorAndFloat("Vector (RGB)"=$Particles.InterpolateOverTime.InterpolateOverTime_V3_.Moving Average, "Float (Alpha)"=$System.Interpolate Internsity)) {isSetParameters=true, module=SetVariables_#7}
   /SetVariables(Particles.R2_DynMatParm:Vector4f=&MakeVector4(
     X=&Multiply_Float(
      A=1.25
      B=&SelectFloatFromArray(
       Direct Array Index=&ReturnExecIndex()
       Float Selection Array=DataInterface() {propertyValues {FloatData [1.5,2,1.3,1]}}
      )
     )
     Y=&Multiply_Float(A=$Engine.Time, B=1)
     Z=&SelectFloatFromArray(
      Direct Array Index=&ReturnExecIndex()
      Float Selection Array=DataInterface() {propertyValues {FloatData [0.25,0.3,0.22,0.2]}}
     )
     W=$Particles.RandU
    )) {isSetParameters=true, module=SetVariables_#8}
   /SetVariables() {isSetParameters=true, module=SetVariables_#9}
   /SetVariables(Particles.VisibilityTag=&MakeCustomIntFromBool(Bool=$User.bisHeadHidden, True INT=1, False INT=0)) {isSetParameters=true, module=SetVariables_#10}
  ]
  rendererClasses [&NiagaraMeshRendererProperties]
  renderers [
   {rendererClass=&NiagaraMeshRendererProperties, rendererIndex=0}
   {rendererClass=&NiagaraMeshRendererProperties, rendererIndex=1}
  ]
 }
}
summary {
 systemName=NS_Skin_BucketKick_Head
 emitters [
  {emitterName=Hair, rendererClasses [&NiagaraMeshRendererProperties]}
  {emitterName=Eye, rendererClasses [&NiagaraMeshRendererProperties]}
  {emitterName=Neck, rendererClasses [&NiagaraMeshRendererProperties]}
  {emitterName=Glow, rendererClasses [&NiagaraSpriteRendererProperties, &NiagaraLightRendererProperties]}
  {emitterName=Flame_bits, rendererClasses [&NiagaraSpriteRendererProperties]}
  {emitterName=HairBig, rendererClasses [&NiagaraMeshRendererProperties]}
 ]
 userVariables [
  User.bIsFrontEnd {type=NiagaraBool}
  User.bIsFrontEndPreview {type=NiagaraBool}
  User.bisHeadHidden {type=NiagaraBool}
 ]
}
